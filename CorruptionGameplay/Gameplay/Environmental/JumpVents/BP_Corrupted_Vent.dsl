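; Corrupted jump vent that launches pawns and vehicles overlapping its sphere, with gravity-scaled velocity and fall-damage immunity.
; The authority handles overlaps, re-checks when a parachuting player lands, and can grant glider redeploy.
; A capsule trace along the launch path damages obstructing objects, sparing traps and spawners.
; The glider-redeploy branch also starts the cooldown by clearing Allowlaunch, whose OnRep toggles light, idle effect and audio.
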(event EventBeginPlay
  (Networking|SwitchHasAuthority
    (:Authority
      (Variables|Default|SetAllowlaunch true)
      (if (not (ScalableFloat|GetValueasBool (Variables|Default|GetJumpVentEnabled)))
        (Actor|DestroyActor)))))

(event Custom|LaunchGC
  (params (LaunchedTarget :type Actor%20Object%20Reference))
  (bind _outcueparameters (GameplayCue|BuildCueParametersfromSource LaunchedTarget))
  (GameplayCue|ExecuteGameplayCue%28Burst%29Local LaunchedTarget (Variables|Default|GetLaunchGameplayCueBurst) _outcueparameters) ; Launching SFX,  Camera shake and Post FX
)

(event OnComponentBeginOverlap(Sphere)
  (params (OverlappedComponent :type Primitive%20Component%20Object%20Reference) (OtherActor :type Actor%20Object%20Reference) (OtherComp :type Primitive%20Component%20Object%20Reference) (OtherBodyIndex :type Integer) (bFromSweep :type Boolean) (SweepResult :type Hit%20Result%20Structure%20%28by%20ref%29))
  (Networking|SwitchHasAuthority
    (:Authority
      (if (Variables|Default|GetAllowlaunch)
        (bind _asfort_player_pawn (Utilities|Casting|CastToFortPlayerPawn OtherActor)
          (:then
            (if (Parachute|IsParachuteOpen _asfort_player_pawn)
              (Character|BindEventtoMovementModeChangedDelegate _asfort_player_pawn (EventDispatchers|CreateEvent :!event Custom|PawnMovementModeChanged))
              (else
                _)))
          (:CastFailed
            _))
        (for _array_element (Collision|GetOverlappingActors (Variables|Default|GetLaunchSphere) "/Script/Engine.Actor" :!class /Script/Engine.PrimitiveComponent)
          (bind _asfort_pawn (Utilities|Casting|CastToFortPawn _array_element)
            (:then
              (Environment|ApplyGlobalEnvironmentGameplayEffecttoActor _asfort_pawn "/CorruptionGameplay/Gameplay/Environmental/JumpVents/GE_Athena_Corrupted_Vent_FX.GE_Athena_Corrupted_Vent_FX_C" 1) ; Apply GE to handle Application and removal of FX (Fixes FORT-224405)
              (CallFunction|LaunchGC _asfort_pawn)
              (bind _asfort_aipawn (Utilities|Casting|CastToFortAIPawn _asfort_pawn)
                (:then
                  (Components|Movement|StopMovementImmediately (Variables|/Script/Engine.Character|GetCharacterMovement _asfort_aipawn)))
                (:CastFailed
                  _))
              (if (not (GameplayTags|HasMatchingGameplayTag (Ability|GetAbilitySystemComponent _asfort_pawn) "(TagName=\"Gameplay.Action.Player.PersonalVehicle\")"))
                (Character|LaunchCharacter _asfort_pawn (CallFunction|CalculateFortPawnLaunch _asfort_pawn) false true) ; Launch for player/AI pawns
                (Variables|/Script/FortniteGame.FortPawn|SetbIgnoreNextFallingDamage true _asfort_pawn) ; Ignore Next Fall Damage
                ;[1 1 1] Destroy Obstructing Building Pieces
                (CallFunction|DestroyBuildingProps _asfort_pawn (Ability|GetAbilitySystemComponent _asfort_pawn) (CallFunction|CalculateFortPawnLaunch _asfort_pawn))
                ;[1 1 1] Glider Redeploy
                (if (Class|AthenaFunctionLibrary|RowtoBool (Variables|Default|GetGliderRedeploy))
                  (if (Math|Boolean|ANDBoolean (not (GameplayTags|HasMatchingGameplayTag (Ability|GetAbilitySystemComponent _asfort_pawn) (Variables|Default|GetGELaunchAppliedTemp))) (Utilities|IsValid _asfort_pawn :!class /Script/Engine.KismetSystemLibrary) (Utilities|IsValidClass (Variables|Default|GetGA_GliderRedeploy)))
                    (bind _asfort_ability_system_component (Utilities|Casting|CastToFortAbilitySystemComponent (Ability|GetAbilitySystemComponent _asfort_pawn))
                      (:then
                        (Ability|Advanced|GrantandActivateAbilityonSelf _asfort_ability_system_component (Variables|Default|GetGA_GliderRedeploy) self))
                      (:CastFailed
                        _))
                    (CallFunction|Launch _asfort_pawn)
                    (Variables|Default|SetAllowlaunch false)
                    (Utilities|Time|SetTimerbyEvent (delegate Custom|AllowReset) (ScalableFloat|GetValueatLevel (Variables|Default|GetLaunchCooldown)))
                    (else
                      _))
                  (else
                    _))))
            (:CastFailed
              (bind _asfort_vehicle_interface (Utilities|Casting|CastToFortVehicleInterface _array_element))
              (Audio|BroadcastAbilityImpactatLocation 0 (Vehicle|GetActor _asfort_vehicle_interface) (CallFunction|CalculateVehicleLaunch _asfort_vehicle_interface (Vehicle|GetActor _asfort_vehicle_interface)) 777.0) ; Launch for vehicles
              (Vehicle|SetIgnoreAllFallingDamage _asfort_vehicle_interface true)
              (bind _returnvalue (Vehicle|GetVehicleActor _asfort_vehicle_interface :!class /Script/FortniteGame.FortVehicleInterface))
              (CallFunction|DestroyBuildingProps _returnvalue (Ability|GetAbilitySystemComponent _returnvalue) (CallFunction|CalculateVehicleLaunch _asfort_vehicle_interface _returnvalue)))))))))

(event Custom|AllowReset
  (Variables|Default|SetAllowlaunch true))

(event Custom|PawnMovementModeChanged
  (params (Character :type Character%20Object%20Reference) (PrevMovementMode :type EMovementMode%20Enum :default "MOVE_None") (PreviousCustomMode :type Byte))
  (if (Utilities|Enum|Equal(Enum) (Variables|/Script/Engine.CharacterMovementComponent|GetMovementMode (Variables|/Script/Engine.Character|GetCharacterMovement Character)) "MOVE_Walking")
    (Networking|SwitchHasAuthority
      (:Authority
        (if (Variables|Default|GetAllowlaunch)
          (for _array_element (Collision|GetOverlappingActors (Variables|Default|GetLaunchSphere) "/Script/Engine.Actor" :!class /Script/Engine.PrimitiveComponent)
            (bind _asfort_pawn (Utilities|Casting|CastToFortPawn _array_element)
              (:then
                (Environment|ApplyGlobalEnvironmentGameplayEffecttoActor _asfort_pawn "/CorruptionGameplay/Gameplay/Environmental/JumpVents/GE_Athena_Corrupted_Vent_FX.GE_Athena_Corrupted_Vent_FX_C" 1) ; Apply GE to handle Application and removal of FX (Fixes FORT-224405)
                (CallFunction|LaunchGC _asfort_pawn)
                (bind _asfort_aipawn (Utilities|Casting|CastToFortAIPawn _asfort_pawn)
                  (:then
                    (Components|Movement|StopMovementImmediately (Variables|/Script/Engine.Character|GetCharacterMovement _asfort_aipawn)))
                  (:CastFailed
                    _))
                (if (not (GameplayTags|HasMatchingGameplayTag (Ability|GetAbilitySystemComponent _asfort_pawn) "(TagName=\"Gameplay.Action.Player.PersonalVehicle\")"))
                  (Character|LaunchCharacter _asfort_pawn (CallFunction|CalculateFortPawnLaunch _asfort_pawn) false true) ; Launch for player/AI pawns
                  (Variables|/Script/FortniteGame.FortPawn|SetbIgnoreNextFallingDamage true _asfort_pawn) ; Ignore Next Fall Damage
                  ;[1 1 1] Destroy Obstructing Building Pieces
                  (CallFunction|DestroyBuildingProps _asfort_pawn (Ability|GetAbilitySystemComponent _asfort_pawn) (CallFunction|CalculateFortPawnLaunch _asfort_pawn))
                  ;[1 1 1] Glider Redeploy
                  (if (Class|AthenaFunctionLibrary|RowtoBool (Variables|Default|GetGliderRedeploy))
                    (if (Math|Boolean|ANDBoolean (not (GameplayTags|HasMatchingGameplayTag (Ability|GetAbilitySystemComponent _asfort_pawn) (Variables|Default|GetGELaunchAppliedTemp))) (Utilities|IsValid _asfort_pawn :!class /Script/Engine.KismetSystemLibrary) (Utilities|IsValidClass (Variables|Default|GetGA_GliderRedeploy)))
                      (bind _asfort_ability_system_component (Utilities|Casting|CastToFortAbilitySystemComponent (Ability|GetAbilitySystemComponent _asfort_pawn))
                        (:then
                          (Ability|Advanced|GrantandActivateAbilityonSelf _asfort_ability_system_component (Variables|Default|GetGA_GliderRedeploy) self))
                        (:CastFailed
                          _))
                      (CallFunction|Launch _asfort_pawn)
                      (Variables|Default|SetAllowlaunch false)
                      (Utilities|Time|SetTimerbyEvent (delegate Custom|AllowReset) (ScalableFloat|GetValueatLevel (Variables|Default|GetLaunchCooldown)))
                      (else
                        _))
                    (else
                      _))))
              (:CastFailed
                (bind _asfort_vehicle_interface (Utilities|Casting|CastToFortVehicleInterface _array_element))
                (Audio|BroadcastAbilityImpactatLocation 0 (Vehicle|GetActor _asfort_vehicle_interface) (CallFunction|CalculateVehicleLaunch _asfort_vehicle_interface (Vehicle|GetActor _asfort_vehicle_interface)) 777.0) ; Launch for vehicles
                (Vehicle|SetIgnoreAllFallingDamage _asfort_vehicle_interface true)
                (bind _returnvalue (Vehicle|GetVehicleActor _asfort_vehicle_interface :!class /Script/FortniteGame.FortVehicleInterface))
                (CallFunction|DestroyBuildingProps _returnvalue (Ability|GetAbilitySystemComponent _returnvalue) (CallFunction|CalculateVehicleLaunch _asfort_vehicle_interface _returnvalue))))))))))

(event OnComponentEndOverlap(Sphere)
  (params (OverlappedComponent :type Primitive%20Component%20Object%20Reference) (OtherActor :type Actor%20Object%20Reference) (OtherComp :type Primitive%20Component%20Object%20Reference) (OtherBodyIndex :type Integer))
  (bind _asfort_player_pawn (Utilities|Casting|CastToFortPlayerPawn OtherActor))
  (Character|UnbindEventfromMovementModeChangedDelegate _asfort_player_pawn (EventDispatchers|CreateEvent :!event Custom|PawnMovementModeChanged)))

(event Custom|Launch
  (params (Target :type Actor%20Object%20Reference))
  (Networking|SwitchHasAuthority
    (:Remote
      (Components|Activation|Activate (Variables|Default|GetLaunchEffect) true)
      (Audio|SpawnSoundAttached "/CorruptionGameplay/Sounds/CorruptedVent/Corrupted_JumpVent_Launch_Cue.Corrupted_JumpVent_Launch_Cue" (Variables|/Script/Engine.Actor|GetRootComponent Target)))))

(fn CalculateFortPawnLaunch
  (params (Actor :type Actor%20Object%20Reference))
  (bind (_bsuccessfullyfoundattribute _returnvalue) (Ability|Attribute|GetFloatAttribute Actor "(AttributeName=\"GravityZScale\",Attribute=/Script/FortniteGame.FortMovementSet:GravityZScale,AttributeOwner=None)"))
  (return (Math|Vector|MakeVector (.x (* (* (Transformation|GetActorForwardVector Actor) (ScalableFloat|GetValueatLevel (Variables|Default|GetForwardAmount))) (Variables|LaunchMultipliers|GetInstanceForwardMultiplier))) (.y (* (* (Transformation|GetActorForwardVector Actor) (ScalableFloat|GetValueatLevel (Variables|Default|GetForwardAmount))) (Variables|LaunchMultipliers|GetInstanceForwardMultiplier))) (* (* (ScalableFloat|GetValueatLevel (Variables|Default|GetLaunchAmount)) (Math|Float|Sqrt _returnvalue)) (Variables|LaunchMultipliers|GetInstanceZ_Multiplier)))))

(fn CalculateVehicleLaunch
  (params (Interface :type Fort%20Vehicle%20Interface%20Interface) (Actor :type Actor%20Object%20Reference))
  (bind (_bsuccessfullyfoundattribute _returnvalue) (Ability|Attribute|GetFloatAttribute Actor "(AttributeName=\"VehicleGravityScale\",Attribute=/Script/FortniteGame.FortVehicleMovementSet:VehicleGravityScale,AttributeOwner=None)"))
  (return (* (+ (Vehicle|GetVehicleVelocity Interface) (* (Math|Vector|MakeVector (.x (* (* (Transformation|GetActorForwardVector Actor) (ScalableFloat|GetValueatLevel (Variables|Default|GetVehicleForwardAmount))) (Variables|LaunchMultipliers|GetInstanceForwardMultiplier))) (.y (* (* (Transformation|GetActorForwardVector Actor) (ScalableFloat|GetValueatLevel (Variables|Default|GetVehicleForwardAmount))) (Variables|LaunchMultipliers|GetInstanceForwardMultiplier))) (* (ScalableFloat|GetValueatLevel (Variables|Default|GetLaunchAmount)) (Variables|LaunchMultipliers|GetInstanceZ_Multiplier))) (ScalableFloat|GetValueatLevel (Variables|Default|GetVehicleImpulseMultiplier)))) (Math|Float|Sqrt _returnvalue))))

(fn DestroyBuildingProps
  (params (Actor :type Actor%20Object%20Reference) (AbilitySystemComp :type Ability%20System%20Component%20Object%20Reference) (LaunchVelocity :type Vector))
  (if (> (ScalableFloat|GetValueatLevel (Variables|Default|GetShouldDestroy)) 0.0)
    (bind _outhits (Collision|MultiCapsuleTraceForObjects :Start (Transformation|GetActorLocation Actor) :End (+ (Transformation|GetActorLocation Actor) (* (Math|Vector|Normalize LaunchVelocity) (ScalableFloat|GetValueatLevel (Variables|Default|GetDestroyDistance)))) :Radius 8e+01 :HalfHeight 1.5e+02 :ObjectTypes (Variables|Default|GetDestroyObjectTypes) :bTraceComplex false :DrawDebugType "None" :bIgnoreSelf true :TraceColor "(R=1.000000,G=0.000000,B=0.000000,A=1.000000)" :TraceHitColor "(R=0.000000,G=1.000000,B=0.000000,A=1.000000)" :DrawTime 8.0))
    (for _array_element _outhits
      (if (Variables|/Script/Engine.Actor|GetbCanBeDamaged (.HitActor (Collision|BreakHitResult _array_element)))
        (bind _bdidactorhavetags (Class|AthenaQuestFunctionLibrary|AthenaCheckActorHasTags (.HitActor (Collision|BreakHitResult _array_element)) "(GameplayTags=((TagName=\"Deimos.Spawner\")))"))
        (if (not _bdidactorhavetags)
          (Utilities|Casting|CastToBuildingTrap (.HitActor (Collision|BreakHitResult _array_element))
            (:CastFailed
              (GameplayEffects|ApplyGameplayEffectToTarget :self AbilitySystemComp :GameplayEffectClass "/Game/Athena/Items/Consumables/ShockwaveGrenade/GE_ShockGrenade_Damage.GE_ShockGrenade_Damage_C" :Target (Ability|GetAbilitySystemComponent (.HitActor (Collision|BreakHitResult _array_element))) :Level 0.0))))))))

(fn IsPlayerRedeployEnabled
  (return (Class|AthenaFunctionLibrary|RowtoBool (Variables|Default|GetGliderRedeploy))))

(fn OnRep_Allowlaunch
  (if (Variables|Default|GetAllowlaunch)
    (Rendering|SetVisibility (Variables|Default|GetPointLight) true false)
    (Components|Activation|Activate (Variables|Default|GetIdleEffect) false)
    (Audio|Components|Audio|FadeIn (Variables|Default|GetAmbientAudio) 1.0 1.0 0.0 "Linear" :!class /Script/Engine.AudioComponent)
    (Audio|PlaySoundatLocation "/CorruptionGameplay/Sounds/CorruptedVent/Corrupted_JumpVent_Reactivate_Cue.Corrupted_JumpVent_Reactivate_Cue" (Transformation|GetWorldLocation (Variables|Default|GetAmbientAudio)))
    (else
      (Rendering|SetVisibility (Variables|Default|GetPointLight) false false)
      (Components|Activation|Deactivate (Variables|Default|GetIdleEffect))
      (Audio|Components|Audio|FadeOut (Variables|Default|GetAmbientAudio) 1.0 0.0 "Linear" :!class /Script/Engine.AudioComponent))))

(fn ConstructionScript)
